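# Physics asset for the Red Oasis backpack cosmetic, supplying collision and joint limits for its two bag straps and a hanging keychain.
# Capsule bodies cover the pack, each strap's root and segment, and a three-body keychain chain.
# Each strap segment swings narrowly on its root with twist locked, while the keychain links swing wider with slight twist.
# No constraint links the pack capsule or the roots to one another.
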PhysicsAsset /BRCosmetics/Accessories/FORT_Backpacks/Backpack_F_MED_RedOasis/Meshes/F_MED_RedOasis_Pack_Physics
bodies {
 dyn_bag_strap_left_01 [{
   shapeType=Capsule
   length=2.6606
   radius=1.0573
   shapeName=dyn_bag_strap_left_01_capsule
   center=<2.3396 0 0>
   rotation=<rot 0 -82.3106 -89.8222>
  }]
 dyn_bag_strap_left_root [{shapeType=Capsule, length=0.505, radius=0.505, shapeName=dyn_bag_strap_left_root_capsule, rotation=<rot 0 -90 0>}]
 dyn_bag_strap_right_01 [{
   shapeType=Capsule
   length=2.6134
   radius=1.0634
   shapeName=dyn_bag_strap_right_01_capsule
   center=<-2.3568 -0.079 -0.097>
   rotation=<rot 0 89.7422 -87.4226>
  }]
 dyn_bag_strap_right_root [{shapeType=Capsule, length=0.505, radius=0.505, shapeName=dyn_bag_strap_right_root_capsule, rotation=<rot 0 -90 0>}]
 dyn_keychain_01 [{shapeType=Capsule, length=1, radius=1, shapeName=dyn_keychain_01_capsule, center=<0.6749 -0.0063 0.0111>, rotation=<rot 0.6976 -90 -1.1054>}]
 dyn_keychain_02 [{shapeType=Capsule, length=1.75, radius=1.255, shapeName=dyn_keychain_02_capsule, center=<1.9855 0 0>, rotation=<rot 0 90 90>}]
 dyn_keychain_root [{shapeType=Capsule, length=0, radius=0.755, shapeName=dyn_keychain_root_capsule, rotation=<rot 0 -90 0>}]
 pack [{shapeType=Capsule, length=7.25, radius=9.75, shapeName=pack_capsule, center=<0 -4.7716 -6.4185>}]
}
constraints [
 {
  bone1Name=dyn_bag_strap_left_01
  bone2Name=dyn_bag_strap_left_root
  swing1LimitDegrees=15
  swing2LimitDegrees=15
  twistLimitDegrees=20
  twistMotion=Locked
 }
 {
  bone1Name=dyn_bag_strap_right_01
  bone2Name=dyn_bag_strap_right_root
  swing1LimitDegrees=15
  swing2LimitDegrees=15
  twistLimitDegrees=20
  twistMotion=Locked
 }
 {bone1Name=dyn_keychain_02, bone2Name=dyn_keychain_01, swing1LimitDegrees=60, swing2LimitDegrees=40, twistLimitDegrees=5}
 {bone1Name=dyn_keychain_01, bone2Name=dyn_keychain_root, swing1LimitDegrees=45, swing2LimitDegrees=20, twistLimitDegrees=5}
]
